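; Item primitive gameplay ability that applies a configurable gameplay effect to its owner, with events geared to consumable heals.
; Activation starts a montage task and immediately runs the internal trigger without waiting for the montage's Triggered pin.
; A successful commit calls force feedback (empty function body here), applies the effect and sends Event.ConsumedHeal to the avatar.
; Only montage cancellation or a failed commit explicitly ends the ability.
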(event Ability|EventActivateAbility
  (Ability|Tasks|PlayMontageInfoAndWaitForTrigger "None" (Variables|Animation|GetMontageInfo) (Variables|Default|GetMontageTargetDataPolicy)
    (:Triggered
      _)
    (:Cancelled
      (Ability|EndAbility))
    (:Completed
      _))
  (CallFunction|TriggerAbilityInternal))

(event Ability|EventOnEndAbility
  (params (bWasCancelled :type Boolean)))

(event Custom|TriggerAbilityInternal
  (bind _returnvalue (Ability|GetAvatarActorfromActorInfo))
  (bind _returnvalue_2 (Ability|CommitAbility))
  (if _returnvalue_2
    (CallFunction|PlayForceFeedback _returnvalue (Variables|Default|GetForceFeedback))
    (Ability|ApplyGameplayEffectToOwner (Variables|Default|GetGE%20Effect%20To%20Apply) 0)
    (Ability|SendGameplayEventtoActor _returnvalue "(TagName=\"Event.ConsumedHeal\")")
    (else
      (Ability|EndAbility))))

(fn PlayForceFeedback
  (params (TargetActor :type Actor%20Object%20Reference) (ForceFeedback :type Force%20Feedback%20Effect%20Object%20Reference)))
