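; ItemAbility_ApplyEffectInRadius is an item ability definition that configures the GA_ItemPrimitive_ApplyEffectInRadius gameplay ability, a primitive for applying an effect in a radius.
; Its Construct function casts the supplied ability to that primitive class and copies its own settings onto it.
; These settings are a montage with play rate, the gameplay effect class taken from an item ability effect asset, force feedback and radius.
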(fn Construct
  (params (Ability :type Gameplay%20Ability%20Object%20Reference))
  (bind _asga_item_primitive_apply_effect_in_radius (Utilities|Casting|CastToGA_ItemPrimitive_ApplyEffectInRadius Ability))
  (Variables|/Script/FortniteGame.FortGameplayAbility|SetMontageInfo (Utilities|Struct|MakeFortGameplayAbilityMontageInfo (Variables|Default|GetAnimMontage) (Variables|Default|GetAnimPlayRate)) _asga_item_primitive_apply_effect_in_radius)
  (Variables|/Itemization/Abilities/GA_ItemPrimitive_ApplyEffectInRadius.GA_ItemPrimitive_ApplyEffectInRadius_C|SetGE%20Effect%20To%20Apply (Variables|/Script/ItemizationCoreRuntime.ItemAbilityEffectAsset|GetGameplayEffectClass (Variables|Default|GetItemAbilityEffect)) _asga_item_primitive_apply_effect_in_radius)
  (Variables|/Itemization/Abilities/GA_ItemPrimitive_ApplyEffectInRadius.GA_ItemPrimitive_ApplyEffectInRadius_C|SetForceFeedback (Variables|Default|GetForceFeedback) _asga_item_primitive_apply_effect_in_radius)
  (Variables|/Itemization/Abilities/GA_ItemPrimitive_ApplyEffectInRadius.GA_ItemPrimitive_ApplyEffectInRadius_C|SetRadius (Variables|Default|GetRadius) _asga_item_primitive_apply_effect_in_radius))
